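# Sound Cue for the DillHail consumable's consume sound, playing a random variation through the consumables sound class and consumable-use attenuation.
# A Quality Level root routes the first two tiers to an equally weighted six-wave Random node and the last two to a three-wave subset, trimming variations by quality.
# A concurrency override caps simultaneous instances, stopping the oldest when the limit is exceeded.
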SoundCue /WildEstateConsumables/Gameplay/DillHail/Audio/SC_DillHail_Consume
root=QualityLevel_0
nodes [
 {id=QualityLevel_0, children [Random_1,Random_1,Random_0,Random_0]}
 {id=Random_0, children [WavePlayer_0,WavePlayer_3,WavePlayer_4], properties {weights [1,1,1]}}
 {id=Random_1, children [WavePlayer_0,WavePlayer_3,WavePlayer_4,WavePlayer_1,WavePlayer_2,WavePlayer_5], properties {weights [1,1,1,1,1,1]}}
 {id=WavePlayer_0, properties {soundWaveAssetPtr=&/DillHail_Consume_03}}
 {id=WavePlayer_1, properties {soundWaveAssetPtr=&/DillHail_Consume_04}}
 {id=WavePlayer_2, properties {soundWaveAssetPtr=&/DillHail_Consume_05}}
 {id=WavePlayer_3, properties {soundWaveAssetPtr=&/DillHail_Consume_02}}
 {id=WavePlayer_4, properties {soundWaveAssetPtr=&/DillHail_Consume_01}}
 {id=WavePlayer_5, properties {soundWaveAssetPtr=&/DillHail_Consume_06}}
]
properties {
 attenuationSettings=&Fort_Consumable_Use_Athena_Att
 bOverrideConcurrency=true
 maxDistance=2250
 soundClassObject=&Fort_Consumables
 volumeMultiplier=1
 concurrencyOverrides {maxCount=4, resolutionRule=StopOldest}
}
