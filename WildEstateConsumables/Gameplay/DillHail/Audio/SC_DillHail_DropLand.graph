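# Sound cue for the Dill Hail consumable landing when dropped.
# A Random node picks one of three equally weighted drop-land wave variants, then a Modulator at default settings adds slight pitch and volume variation.
# The cue plays at reduced volume under the loot collect sound class, using the shared Fortnite weapon-drop attenuation.
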SoundCue /WildEstateConsumables/Gameplay/DillHail/Audio/SC_DillHail_DropLand
root=Modulator_2
nodes [
 {id=Modulator_2, children [Random_0]}
 {id=Random_0, children [WavePlayer_1,WavePlayer_2,WavePlayer_4], properties {weights [1,1,1]}}
 {id=WavePlayer_1, properties {soundWaveAssetPtr=&/DillHail_DropLand_01}}
 {id=WavePlayer_2, properties {soundWaveAssetPtr=&/DillHail_DropLand_02}}
 {id=WavePlayer_4, properties {soundWaveAssetPtr=&/DillHail_DropLand_03}}
]
properties {
 attenuationSettings=&Fort_Weapon_Drop_Att
 maxDistance=3000
 soundClassObject=&Fort_Loot_Collect
 volumeMultiplier=0.5
}
