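# Sound cue for the DillHail consumable's heal sound.
# A Random node picks one of three heal wave variations with equal weight.
# It uses the Fortnite consumables sound class and consumable-use attenuation, with a concurrency override that caps simultaneous instances and stops the oldest.
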SoundCue /WildEstateConsumables/Gameplay/DillHail/Audio/SC_DillHail_Heal
root=Random_1
nodes [
 {id=Random_1, children [WavePlayer_2,WavePlayer_5,WavePlayer_6], properties {weights [1,1,1]}}
 {id=WavePlayer_2, properties {soundWaveAssetPtr=&/DillHail_Heal_01}}
 {id=WavePlayer_5, properties {soundWaveAssetPtr=&/DillHail_Heal_02}}
 {id=WavePlayer_6, properties {soundWaveAssetPtr=&/DillHail_Heal_03}}
]
properties {
 attenuationSettings=&Fort_Consumable_Use_Athena_Att
 bOverrideConcurrency=true
 maxDistance=2250
 soundClassObject=&Fort_Consumables
 volumeMultiplier=1
 concurrencyOverrides {maxCount=4, resolutionRule=StopOldest}
}
