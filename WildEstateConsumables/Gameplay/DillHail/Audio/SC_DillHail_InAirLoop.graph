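# DillHail in-air loop SoundCue in the gadgets sound class, reusing PizzaParty box in-air audio.
# A Random node picks one of two equally weighted looping waves and feeds a Modulator that randomizes pitch upward, on top of raised cue pitch and volume.
# Inline attenuation overrides the referenced ToyHit asset with logarithmic falloff.
# Inline concurrency limits simultaneous instances per owner.
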SoundCue /WildEstateConsumables/Gameplay/DillHail/Audio/SC_DillHail_InAirLoop
root=Modulator_1
nodes [
 {id=Modulator_1, children [Random_0], properties {pitchMax=1.2, pitchMin=1}}
 {id=Random_0, children [WavePlayer_0,WavePlayer_3], properties {weights [1,1]}}
 {id=WavePlayer_0, properties {bLooping=true, soundWaveAssetPtr=&/PizzaParty_BoxInAirLoop_01}}
 {id=WavePlayer_3, properties {bLooping=true, soundWaveAssetPtr=&/PizzaParty_BoxInAirLoop_02}}
]
properties {
 attenuationSettings=/Game/Sounds/Attenuation/Fort_ToyHit_Att
 bOverrideAttenuation=true
 bOverrideConcurrency=true
 maxDistance=2000
 pitchMultiplier=1.2
 soundClassObject=&Fort_Gadgets
 volumeMultiplier=1.2
 attenuationOverrides {distanceAlgorithm=Logarithmic, falloffDistance=2000, stereoSpread=512, attenuationShapeExtents {x=0}}
 concurrencyOverrides {bLimitToOwner=true, maxCount=2, volumeScale=0.7}
}
