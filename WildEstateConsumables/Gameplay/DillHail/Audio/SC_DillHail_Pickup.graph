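# Sound cue for picking up the DillHail consumable, classed as a loot collect sound.
# A Random node picks one of three equally weighted pickup waves and feeds a Modulator at the root.
# Playback uses weapon pickup attenuation at reduced volume.
# Overridden concurrency allows one instance per owner and blocks new plays while one is active.
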SoundCue /WildEstateConsumables/Gameplay/DillHail/Audio/SC_DillHail_Pickup
root=Modulator_1
nodes [
 {id=Modulator_1, children [Random_0]}
 {id=Random_0, children [WavePlayer_0,WavePlayer_3,WavePlayer_4], properties {weights [1,1,1]}}
 {id=WavePlayer_0, properties {soundWaveAssetPtr=&/DillHail_Pickup_01}}
 {id=WavePlayer_3, properties {soundWaveAssetPtr=&/DillHail_Pickup_02}}
 {id=WavePlayer_4, properties {soundWaveAssetPtr=&/DillHail_Pickup_03}}
]
properties {
 attenuationSettings=&Fort_Weapon_Pickup_Att
 bOverrideConcurrency=true
 maxDistance=3500
 soundClassObject=&Fort_Loot_Collect
 volumeMultiplier=0.3
 concurrencyOverrides {bLimitToOwner=true, maxCount=1, resolutionRule=PreventNew}
}
