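# Sound cue for the DillHail consumable's ready foley, reusing PizzaParty pre-throw sounds as a quiet weapon-foley cue with reload attenuation.
# A mixer layers two random picks: one of two slice sounds, pitched down and mixed at lower volume, plus one of two box sounds at full input volume.
# A root modulator adds slight pitch variation to the combined result.
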SoundCue /WildEstateConsumables/Gameplay/DillHail/Audio/SC_DillHail_Ready
root=Modulator_0
nodes [
 {id=Mixer_0, children [Modulator_1,Random_3], properties {inputVolume [0.6,1]}}
 {id=Modulator_0, children [Mixer_0], properties {pitchMax=1.02, pitchMin=0.98}}
 {id=Modulator_1, children [Random_2], properties {pitchMax=0.9, pitchMin=0.8}}
 {id=Random_2, children [WavePlayer_1,WavePlayer_3], properties {weights [1,1]}}
 {id=Random_3, children [WavePlayer_8,WavePlayer_9], properties {weights [1,1]}}
 {id=WavePlayer_1, properties {soundWaveAssetPtr=&/PizzaParty_SlicePreThrow_01}}
 {id=WavePlayer_3, properties {soundWaveAssetPtr=&/PizzaParty_SlicePreThrow_02}}
 {id=WavePlayer_8, properties {soundWaveAssetPtr=&/PizzaParty_BoxPreThrow_03}}
 {id=WavePlayer_9, properties {soundWaveAssetPtr=&/PizzaParty_BoxPreThrow_04}}
]
properties {
 attenuationSettings=&Fort_Weapon_Reload_Att
 maxDistance=2250
 soundClassObject=&Fort_WeaponFoley
 volumeMultiplier=0.2
}
