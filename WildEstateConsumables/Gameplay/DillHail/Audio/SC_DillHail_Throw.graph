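# Sound cue for the DillHail consumable's throw, built from reused PizzaParty throw recordings.
# A mixer blends a briefly delayed random box-throw wave with a quieter random slice-throw layer that has its own upward pitch variation.
# An outer modulator adds further pitch variation over the mix, and the cue plays pitched up at reduced volume.
# It uses gadget-place attenuation and a concurrency override that prevents new instances at the limit.
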SoundCue /WildEstateConsumables/Gameplay/DillHail/Audio/SC_DillHail_Throw
root=Modulator_0
nodes [
 {id=Delay_0, children [Random_0], properties {delayMax=0.03, delayMin=0.03}}
 {id=Mixer_0, children [Delay_0,Modulator_1], properties {inputVolume [1,0.6]}}
 {id=Modulator_0, children [Mixer_0], properties {pitchMax=1.1, pitchMin=0.9}}
 {id=Modulator_1, children [Random_1], properties {pitchMax=1.2, pitchMin=1}}
 {id=Random_0, children [WavePlayer_6,WavePlayer_7,WavePlayer_8], properties {weights [1,1,1]}}
 {id=Random_1, children [WavePlayer_9, WavePlayer_10, WavePlayer_11], properties {weights [1,1,1]}}
 {id=WavePlayer_10, properties {soundWaveAssetPtr=&/PizzaParty_SliceThrowLayer_02}}
 {id=WavePlayer_11, properties {soundWaveAssetPtr=&/PizzaParty_SliceThrowLayer_03}}
 {id=WavePlayer_6, properties {soundWaveAssetPtr=&/PizzaParty_BoxThrow_01}}
 {id=WavePlayer_7, properties {soundWaveAssetPtr=&/PizzaParty_BoxThrow_02}}
 {id=WavePlayer_8, properties {soundWaveAssetPtr=&/PizzaParty_BoxThrow_03}}
 {id=WavePlayer_9, properties {soundWaveAssetPtr=&/PizzaParty_SliceThrowLayer_01}}
]
properties {
 attenuationSettings=/Game/Sounds/Attenuation/Fort_GadgetPlace_Att
 bOverrideConcurrency=true
 maxDistance=2000
 pitchMultiplier=1.5
 soundClassObject=&Fort_Gadgets_Items
 volumeMultiplier=0.5
 concurrencyOverrides {resolutionRule=PreventNew}
}
