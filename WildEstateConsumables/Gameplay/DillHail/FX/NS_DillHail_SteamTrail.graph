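# Niagara system for the steam trail of the DillHail consumable, built from soft sprite puffs that drift upward.
# SteamLoop spawns at a steady rate gated by an owner-speed threshold, while SteamPerUnit emits shorter-lived, faster-rising puffs by distance moved.
# Both use randomized sub-image frames and fade alpha with distance from their source; SteamLoop also tints red to blue.
# The SmokeRibbon emitter and the skeletal mesh sampling in both steam emitters are disabled.
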NiagaraSystem /WildEstateConsumables/Gameplay/DillHail/FX/NS_DillHail_SteamTrail
dependencies {
 usedDataInterfaces [
  &NiagaraDataInterfaceCurve
  /Script/Niagara.NiagaraDataInterfaceSimpleCounter
  &NiagaraDataInterfaceCamera
  &NiagaraDataInterfaceColorCurve
 ]
 usedDynamicInputs []
 usedModules [&SystemState]
}
emitters {
 SmokeRibbon {
  bEnabled=false
  emitterUpdateScript [
   &EmitterState()
   &SpawnPerUnit(
    Module.Use Max Movement Threshold=true
    Module.Use Movement Tolerance=false
    Spawn Per Unit=15
    Spawn Probability:NiagaraFloat=&Lerp_Float(A=1, B=0, Alpha=&MakeFloatFromInt(Int=$System.isSkydiving))
    Use Spawn Probability=true
   )
   &SpawnRate(SpawnRate=20, Use Spawn Probability=false) {enabled=false}
  ]
  particleSpawnScript [
   /SetVariables(Particles.Lifetime=0.5, Particles.RibbonWidth=50) {isSetParameters=true}
   &SkeletalMeshLocation(
    Bone / Socket Index=2
    Bone Sampling Mode=::NewEnumerator4
    Inherit Velocity=false
    MaskSpawnGroup=false
    Sampled Position Offset=<-13.1117 -11 2.1285>
    Sampled Position Offset Coordinate Space:ENiagaraMeshTransforms=::NewEnumerator3
    Skeletal Mesh=$User.SkeletalMesh
   ) {enabled=false}
   &AddVelocity(Velocity=<0 -1000 0>) {enabled=false, module=AddVelocity001}
   /SetVariables(Particles.RibbonLinkOrder:NiagaraFloat=&MakeFloatFromInt(Int=/Niagara/DynamicInputs/DataInterface/GenerateIntFromCounter(SimpleCounter:NiagaraDataInterfaceSimpleCounter=DataInterface() {propertyValues {GpuSyncMode=None, InitialValue=0}}))) {isSetParameters=true}
   &InheritVelocity(Module.Limit Speed=false, Velocity Scale=&VectorFromFloat(Value=1)) {enabled=false}
   &SkeletalMeshLocation(
    Apply To Particle Position=false
    Bone Sampling Mode=::NewEnumerator4
    Inherit Velocity=false
    MaskSpawnGroup=false
    Sampled Position Offset=<-13.1117 -11 2.1285>
    Sampled Position Offset Coordinate Space:ENiagaraMeshTransforms=::NewEnumerator3
    Skeletal Mesh=$User.SkeletalMesh
   ) {enabled=false, module=SkeletalMeshLocation001}
   /SetVariables(
    Particles.MeshOrientationTest:Quat4f=$Output.SkeletalMeshLocation001.SampledRotation
    Particles.RibbonFacing=&RotateVectorByQuaternion(Vector To Rotate=<-1 0 0>, Quaternion=$Output.SkeletalMeshLocation001.SampledRotation)
    Particles.RibbonFacingAlt:Vector3f=&RotateVectorByQuaternion(Vector To Rotate=<1 0 0>, Quaternion=$Output.SkeletalMeshLocation001.SampledRotation)
    Particles.RibbonFacingAltAlt:Vector3f=&RotateVectorByQuaternion(Vector To Rotate=<0 0 -1>, Quaternion=$Output.SkeletalMeshLocation001.SampledRotation)
   ) {isSetParameters=true}
  ]
  particleUpdateScript [
   &ParticleState()
   &SkeletalMeshLocation(
    Apply To Particle Position:NiagaraBool=&SetBoolByFloatComparison(A=$Particles.NormalizedAge, B=0.25, Comparison Type=::NewEnumerator6)
    Bone Sampling Mode=::NewEnumerator4
    Inherit Velocity=false
    MaskSpawnGroup=false
    Sampled Position Offset Coordinate Space:ENiagaraMeshTransforms=::NewEnumerator3
    Skeletal Mesh=$User.SkeletalMesh
   ) {module=SkeletalMeshLocation003}
   &Drag(Drag=5, Use Linear Drag=true, Use Rotational Drag=false)
   &GravityForce(Gravity=<0 0 600>)
   &SolveForcesAndVelocity(Clamp Velocity=false, Limit Acceleration=false)
   /Game/Effects/Niagara/Modules/Materials/FN_RibbonMaterial/FN_RibbonMaterial_v2(
    Color:LinearColor=&MakeLinearColorFromVectorAndFloat(
     "Vector (RGB)"=<2 2 2>
     "Float (Alpha)"=&FloatFromCurve(
      FloatCurve=DataInterface() {propertyValues {Curve {keys [
          {arriveTangent=0.0082, interpMode=RCIM_Cubic, leaveTangent=0.0082, tangentMode=RCTM_User, time=0.25, value=0}
          {arriveTangent=-0.0105, interpMode=RCIM_Cubic, leaveTangent=-0.0105, tangentMode=RCTM_User, time=0.4453, value=0.1}
          {arriveTangent=-0.4528, interpMode=RCIM_Cubic, leaveTangent=-0.4528, tangentMode=RCTM_User, time=1, value=0}
         ]}}}
      Scale Curve=&Multiply_Float(
       A=/Niagara/DynamicInputs/Math/LerpMultipleFloats(
        Float Count:ENiagara_MultipleLerpCount=::NewEnumerator0
        1:NiagaraFloat=0.65
        2:NiagaraFloat=1.05
        3:NiagaraFloat=0.65
        4:NiagaraFloat=?
        5:NiagaraFloat=?
        Alpha:NiagaraFloat=Linked() {linkedVariable {name=NPC.FortniteNPC.TimeOfDay_0-1, type=NiagaraFloat}}
       )
       B=5
      )
     )
    )
    Color Pan U Speed=2.8
    Color Scale U=2
    Erosion Amount:NiagaraFloat=&Lerp_Float(A=0, B=1.75, Alpha=$Particles.NormalizedAge)
    Erosion Pan Speed U=2.75
    Erosion Power=0.25
    Erossion Scale U=2
    Opacity Pan U Speed=0.23
    Opacity Scale U=0.2
    Scale Alpha:NiagaraFloat=&FloatFromCurve(
     FloatCurve=DataInterface() {propertyValues {Curve {keys [
         {interpMode=RCIM_Cubic, tangentMode=RCTM_User, time=0, value=1}
         {arriveTangent=-2.8707, interpMode=RCIM_Cubic, leaveTangent=-2.8707, tangentMode=RCTM_User, time=1, value=0}
        ]}}}
     Scale Curve=2
    )
    TIme:NiagaraFloat=$Engine.Time
   )
   &ScaleRibbonWidth(Ribbon Width Scale=&FloatFromCurve(FloatCurve=DataInterface() {propertyValues {Curve {keys [
         {arriveTangent=0.4163, interpMode=RCIM_Cubic, leaveTangent=0.4163, tangentMode=RCTM_User, time=0.3, value=1}
         {arriveTangent=-1.9468, interpMode=RCIM_Cubic, leaveTangent=-1.9468, tangentMode=RCTM_User, time=1.0032, value=0.0313}
        ]}}}))
   &self:GetCameraInfo_01(
    Dot Vector:Vector3f=&MakeCustomVectorFromBool(
     Bool=&SetBoolByFloatComparison(A=/Niagara/DynamicInputs/TypeConversions/VectorLengthSafe(Vector:Vector3f=$Output.SkeletalMeshLocation003.SampledVelocity), B=3)
     True Vector=/Niagara/DynamicInputs/Helpers/SafeNormalizeVector(Vector To Normalize:Vector3f=$Output.SkeletalMeshLocation003.SampledVelocity, Fallback Vector:Vector3f=<0 0 0>)
     False Vector=<0 0 0>
    )
    Dot product to return 1=0.95
    Interpolate Value=5
   ) {module=GetCameraInfo_01}
   &ScaleColor(Color Value To Scale=$Particles.Color, Scale Alpha=$Particles.CameraVelocityDot, ScaleA=true, ScaleRGB=false)
   &Color(Color=&ColorFromCurve(
     ColorCurve=DataInterface() {propertyValues {
       AlphaCurve {keys [<at 0 1 RCIM_Cubic>]}
       BlueCurve {keys [<at 0 0 RCIM_Cubic>,<at 0.3333 0>,<at 0.6667 0>,<at 1 1 RCIM_Cubic>]}
       GreenCurve {keys [<at 0 0 RCIM_Cubic>,<at 0.3333 1>,<at 0.6667 1>,<at 1 0 RCIM_Cubic>]}
       RedCurve {keys [<at 0 1 RCIM_Cubic>,<at 0.3333 1>,<at 0.6667 0>,<at 1 0 RCIM_Cubic>]}
      }}
     CurveIndex=$Particles.CameraVelocityDot
    )) {enabled=false}
  ]
  rendererClasses [&NiagaraRibbonRendererProperties, &NiagaraMeshRendererProperties]
  renderers [
   {rendererClass=&NiagaraRibbonRendererProperties}
   {rendererClass=&NiagaraMeshRendererProperties, rendererIndex=1}
   {rendererClass=&NiagaraRibbonRendererProperties, rendererIndex=2}
   {rendererClass=&NiagaraRibbonRendererProperties, rendererIndex=3}
  ]
 }
 SteamLoop {
  emitterUpdateScript [
   &EmitterState(Life Cycle Mode=::NewEnumerator1, Loop Duration=1, UseLoopDelay=false)
   &SpawnRate(
    Spawn Probability=&MakeCustomFloatFromBool(
     Bool=&SetBoolByFloatComparison(A=&VectorLength(Vector=$Engine.Owner.Velocity), B=5, Comparison Type=::NewEnumerator6)
     True Float=1
     False Float=0
    )
    SpawnRate=5
    Use Spawn Probability=true
   )
  ]
  particleSpawnScript [
   /SetVariables(Particles.SpawnGroup=&SpawnGroup()) {isSetParameters=true, module=SetVariables_#1}
   /Niagara/Modules/Spawn/Initialization/InitializeParticle(
    Color=<rgba 1 1 1 0.1>
    Lifetime=&RandomRangeFloat(Minimum=0.8, Maximum=1.5)
    Module.Write Color=true
    Module.Write Lifetime=true
    Module.Write Mass=false
    Module.Write Material Random=true
    Module.Write Position=false
    Module.Write Scale=false
    Module.Write SpriteRotation=true
    Module.Write SpriteSize=true
    Position=&SimulationPosition()
    Sprite Rotation=&UniformRangedFloat(Minimum=0, Maximum=360)
    Sprite Size=&Vector2DFromFloat(Value=&UniformRangedFloat(Minimum=15, Maximum=30))
   )
   &SphereLocation(
    MaskSpawnGroup=false
    Non Uniform Scale=<1 1 0>
    Offset=<0 0 5>
    Override Seed=false
    Sphere Coordinate Space=::World
    Sphere Radius=5
    Surface Only=true
    UseNonUniformScale=true
    UseOffset=true
    UseSphereOrientation=true
   )
   &SkeletalMeshLocation(
    Inherit Velocity=false
    MaskSpawnGroup=true
    Mesh Sampling Type=::NewEnumerator5
    Sampled Position Offset=&Multiply_Vector(A=<1 1 1>, B=$Output.SphereLocation.SphereVector)
    Sampled Position Offset Coordinate Space:ENiagaraMeshTransforms=::NewEnumerator3
    Skeletal Mesh=$User.SkeletalMesh
    Socket Sampling Mode=::NewEnumerator5
   ) {enabled=false}
   &AddVelocity(Coordinate Space=::World, Velocity=&RandomRangeVector(Minimum=<0 0 2>, Maximum=<1 1 8>))
   &CameraOffset(Camera Offset Amount=5) {enabled=false}
   &InheritVelocity(Module.Limit Speed=true, Velocity Limit=20, Velocity Scale=&VectorFromFloat(Value=0.25))
  ]
  particleUpdateScript [
   &ParticleState()
   &SkeletalMeshLocation(
    Apply To Particle Position=false
    Inherit Velocity=false
    MaskSpawnGroup=false
    Mesh Sampling Type=::NewEnumerator5
    Skeletal Mesh=$User.SkeletalMesh
    Socket Sampling Mode=::NewEnumerator5
   ) {enabled=false, module=SkeletalMeshLocation001}
   /SetVariables(Particles.DistanceFromSource:NiagaraFloat=&ClampFloat(Float=&Divide_Float(
      A=&Add_Float(
       A=0
       B=&VectorLength(
        Vector=&Subtract_Vector(A=Linked() {linkedVariable {name=Particles.Position, type=NiagaraPosition}}, B=$Output.SphereLocation.Position)
        Safe Length Calculation=false
       )
      )
      B=20
     ))) {isSetParameters=true, module=SetVariables_#2}
   &GravityForce(Gravity=<0 0 20>)
   &SolveForcesAndVelocity(Clamp Velocity=false, Limit Acceleration=false)
   &ScaleSpriteSize(Scale Factor=&Vector2DFromFloat(Value=&FloatFromCurve(FloatCurve=DataInterface() {propertyValues {Curve {keys [{arriveTangent=-0.28, interpMode=RCIM_Cubic, leaveTangent=-0.28, tangentMode=RCTM_User, time=0, value=0.5}, <at 1 1>]}}}))) {module=ScaleSpriteSize001}
   &ScaleSpriteSize(Scale Factor=&Vector2DFromFloat(Value=&FloatFromCurve(FloatCurve=DataInterface() {propertyValues {Curve {keys [
          {arriveTangent=-0.0248, interpMode=RCIM_Cubic, leaveTangent=-0.0248, tangentMode=RCTM_User, time=0, value=1}
          {arriveTangent=-0.5458, interpMode=RCIM_Cubic, leaveTangent=-0.5458, tangentMode=RCTM_User, time=1, value=0.75}
         ]}}})))
   &DynamicMaterialParameters(
    Index 0 Param 1=&FloatFromCurve(
     FloatCurve=DataInterface() {propertyValues {Curve {keys [
         {arriveTangent=4.0913, interpMode=RCIM_Cubic, leaveTangent=4.0913, tangentMode=RCTM_User, time=0, value=0}
         {arriveTangent=0.1513, interpMode=RCIM_Cubic, leaveTangent=0.1513, tangentMode=RCTM_User, time=0.5, value=1}
         {arriveTangent=-0.0082, interpMode=RCIM_Cubic, leaveTangent=-0.0082, tangentMode=RCTM_User, time=1, value=0}
        ]}}}
     Scale Curve=0.6
    )
    Index 0 Param 4=1
    Module.Param0WriteEnabled=true
    Module.Param1WriteEnabled=false
    Module.Param2WriteEnabled=false
    Module.Param3WriteEnabled=true
   )
   /SetVariables(Particles.StartFrame:NiagaraFloat=&RandomRangeFloat(Minimum=0, Maximum=63)) {isSetParameters=true, module=SetVariables_#3}
   /SetVariables(Particles.SubImageIndex=&Lerp_Float(
     A=$Particles.StartFrame
     B=&Add_Float(A=$Particles.StartFrame, B=&RandomRangeFloat(Minimum=2, Maximum=6))
     Alpha=$Particles.NormalizedAge
    )) {isSetParameters=true, module=SetVariables_#4}
   &Color(Color=&LerpLinearColors(StartColor=<rgba 1 0 0 1>, EndColor=<rgba 0 0 1 1>, LerpFactor=$Particles.DistanceFromSource))
   &ScaleColor(Scale Alpha=&Lerp_Float(A=1, B=0.2, Alpha=$Particles.DistanceFromSource), ScaleA=true, ScaleRGB=true)
   &KillParticles(Kill Particles=&SetBoolByFloatComparison(
     A=&MakeFloatFromLinearColor(LinearColor=$Particles.Color, Channel:ENiagaraLinearColor_Channels=::NewEnumerator3)
     B=0
     Comparison Type=::NewEnumerator4
    )) {enabled=false}
  ]
  rendererClasses [&NiagaraSpriteRendererProperties]
  renderers [{rendererClass=&NiagaraSpriteRendererProperties}]
 }
 SteamPerUnit {
  emitterUpdateScript [
   &EmitterState(Life Cycle Mode=::NewEnumerator1, Loop Duration=1, UseLoopDelay=false)
   &SpawnRate(SpawnRate=5, Use Spawn Probability=false) {enabled=false}
   &SpawnPerUnit(
    Module.Use Max Movement Threshold=true
    Module.Use Movement Tolerance=true
    Movement Tolerance=4
    Spawn Per Unit=30
    Use Spawn Probability=true
   )
  ]
  particleSpawnScript [
   /SetVariables(Particles.SpawnGroup=&SpawnGroup()) {isSetParameters=true, module=SetVariables_#1}
   &InitializeParticle(
    Color=<rgba 1 1 1 0.05>
    Lifetime=&RandomRangeFloat(Minimum=0.2, Maximum=0.5)
    Module.Write Color=true
    Module.Write Lifetime=true
    Module.Write Mass=false
    Module.Write Material Random=true
    Module.Write Position=false
    Module.Write Scale=false
    Module.Write SpriteRotation=true
    Module.Write SpriteSize=true
    Position=&SimulationPosition()
    Sprite Rotation=&UniformRangedFloat(Minimum=0, Maximum=360)
    Sprite Size=&Vector2DFromFloat(Value=&UniformRangedFloat(Minimum=15, Maximum=30))
   )
   &SphereLocation(
    MaskSpawnGroup=false
    Non Uniform Scale=<1 1 0>
    Override Seed=false
    Sphere Coordinate Space=::World
    Sphere Radius=5
    Surface Only=true
    UseNonUniformScale=true
    UseOffset=false
    UseSphereOrientation=true
   )
   &SkeletalMeshLocation(
    Inherit Velocity=false
    MaskSpawnGroup=true
    Mesh Sampling Type=::NewEnumerator5
    Sampled Position Offset=&Multiply_Vector(A=<1 1 1>, B=$Output.SphereLocation.SphereVector)
    Sampled Position Offset Coordinate Space:ENiagaraMeshTransforms=::NewEnumerator3
    Skeletal Mesh=$User.SkeletalMesh
    Socket Sampling Mode=::NewEnumerator5
   ) {enabled=false}
   &AddVelocity(Coordinate Space=::World, Velocity=&RandomRangeVector(Minimum=<0 0 50>, Maximum=<1 1 80>))
   &CameraOffset(Camera Offset Amount=5) {enabled=false}
   &InheritVelocity(Module.Limit Speed=true, Velocity Limit=80, Velocity Scale=&VectorFromFloat(Value=0.25))
  ]
  particleUpdateScript [
   &ParticleState()
   &SkeletalMeshLocation(
    Apply To Particle Position=false
    Inherit Velocity=false
    MaskSpawnGroup=false
    Mesh Sampling Type=::NewEnumerator5
    Skeletal Mesh=$User.SkeletalMesh
    Socket Sampling Mode=::NewEnumerator5
   ) {enabled=false, module=SkeletalMeshLocation001}
   /SetVariables(Particles.DistanceFromSource:NiagaraFloat=&ClampFloat(Float=&Divide_Float(
      A=&Add_Float(
       A=0
       B=&VectorLength(Vector=&Subtract_Vector(
         A=Linked() {linkedVariable {name=Particles.Position, type=NiagaraPosition}}
         B=Linked() {linkedVariable {name=Engine.Owner.Position, type=NiagaraPosition}}
        ))
      )
      B=80
     ))) {isSetParameters=true, module=SetVariables_#2}
   &GravityForce(Gravity=<0 0 20>)
   &SolveForcesAndVelocity(Clamp Velocity=false, Limit Acceleration=false)
   &ScaleSpriteSize(Scale Factor=&Vector2DFromFloat(Value=&FloatFromCurve(FloatCurve=DataInterface() {propertyValues {Curve {keys [{arriveTangent=-0.28, interpMode=RCIM_Cubic, leaveTangent=-0.28, tangentMode=RCTM_User, time=0, value=0.5}, <at 1 1>]}}}))) {module=ScaleSpriteSize001}
   &ScaleSpriteSize(Scale Factor=&Vector2DFromFloat(Value=&FloatFromCurve(FloatCurve=DataInterface() {propertyValues {Curve {keys [
          {interpMode=RCIM_Cubic, tangentMode=RCTM_User, time=0, value=1}
          {arriveTangent=-2.5886, interpMode=RCIM_Cubic, leaveTangent=-2.5886, tangentMode=RCTM_User, time=1, value=0}
         ]}}}))) {enabled=false}
   &DynamicMaterialParameters(
    Index 0 Param 1=&FloatFromCurve(
     FloatCurve=DataInterface() {propertyValues {Curve {keys [
         {arriveTangent=4.0913, interpMode=RCIM_Cubic, leaveTangent=4.0913, tangentMode=RCTM_User, time=0, value=0}
         {arriveTangent=0.1513, interpMode=RCIM_Cubic, leaveTangent=0.1513, tangentMode=RCTM_User, time=0.5, value=1}
         {arriveTangent=-0.0082, interpMode=RCIM_Cubic, leaveTangent=-0.0082, tangentMode=RCTM_User, time=1, value=0}
        ]}}}
     Scale Curve=0.6
    )
    Index 0 Param 4=1
    Module.Param0WriteEnabled=true
    Module.Param1WriteEnabled=false
    Module.Param2WriteEnabled=false
    Module.Param3WriteEnabled=true
   )
   /SetVariables(Particles.StartFrame:NiagaraFloat=&RandomRangeFloat(Minimum=0, Maximum=63)) {isSetParameters=true, module=SetVariables_#3}
   /SetVariables(Particles.SubImageIndex=&Lerp_Float(
     A=$Particles.StartFrame
     B=&Add_Float(A=$Particles.StartFrame, B=&RandomRangeFloat(Minimum=2, Maximum=6))
     Alpha=$Particles.NormalizedAge
    )) {isSetParameters=true, module=SetVariables_#4}
   &Color(Color=&LerpLinearColors(StartColor=<rgba 1 0 0 1>, EndColor=<rgba 0 0 1 1>, LerpFactor=$Particles.DistanceFromSource)) {enabled=false}
   &ScaleColor(Scale Alpha=&Lerp_Float(A=0.5, B=0.1, Alpha=$Particles.DistanceFromSource), ScaleA=true, ScaleRGB=true)
   &KillParticles(Kill Particles=&SetBoolByFloatComparison(
     A=&MakeFloatFromLinearColor(LinearColor=$Particles.Color, Channel:ENiagaraLinearColor_Channels=::NewEnumerator3)
     B=0
     Comparison Type=::NewEnumerator4
    )) {enabled=false}
  ]
  rendererClasses [&NiagaraSpriteRendererProperties]
  renderers [{rendererClass=&NiagaraSpriteRendererProperties}]
 }
}
summary {
 systemName=NS_DillHail_SteamTrail
 emitters [
  {emitterName=SteamLoop, rendererClasses [&NiagaraSpriteRendererProperties]}
  {
   bEnabled=false
   emitterName=SmokeRibbon
   rendererClasses [&NiagaraRibbonRendererProperties, &NiagaraMeshRendererProperties]
  }
  {emitterName=SteamPerUnit, rendererClasses [&NiagaraSpriteRendererProperties]}
 ]
 userVariables [
  User.RibbonIDOffset {type=NiagaraInt32}
  User.SkeletalMesh {
   type=NiagaraDataInterfaceSkeletalMesh
   defaultValue {
    struct=NiagaraExt_VariableValue_DataInterface
    value {dataInterface=&self:NiagaraDataInterfaceSkeletalMesh_0}
   }
  }
 ]
}
