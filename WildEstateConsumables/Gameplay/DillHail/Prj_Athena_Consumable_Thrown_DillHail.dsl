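; Thrown DillHail consumable projectile that extends its parent class with an in-air sound and long-throw detection.
; On begin play it stores its starting location and plays the looping audio, which stops on hit.
; When its pickup spawns, it compares the straight-line distance from the start against a threshold in meters.
; Throws at or beyond that broadcast the BurgerYeet custom verb message for the instigator's pawn.
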(event EventBeginPlay
  (|CallParentFunction)
  (Variables|Default|SetInitial%20Loc (Projectile|GetCurrentLocation :!class /Script/FortniteGame.FortProjectileMovementInterface))
  (Audio|Components|Audio|Play (Variables|Default|GetAudioInAirLoop) 0.0))

(event EventOnPickupSpawned
  (params (Spawned%20Pickup :type Fort%20Pickup%20Object%20Reference))
  (|CallParentFunction Spawned%20Pickup)
  (Variables|Default|SetLanded%20Loc (Transformation|GetActorLocation Spawned%20Pickup))
  (if (>= (Math|Vector|Distance(Vector) (Variables|Default|GetInitial%20Loc) (Variables|Default|GetLanded%20Loc)) (* 1e+02 (Variables|Default|GetThrowDistanceInMeters)))
    (CustomVerb|BroadcastCustomVerbMessage (CustomVerb|MakeCustomVerbMessage (Game|Player|GetPlayerPawn (Projectile|GetInstigatorPlayerController :!class /Script/FortniteGame.FortProjectileMovementInterface)) "(TagName=\"CustomVerb.BurgerYeet\")"))))

(event Collision|EventHit
  (params (MyComp :type Primitive%20Component%20Object%20Reference) (Other :type Actor%20Object%20Reference) (OtherComp :type Primitive%20Component%20Object%20Reference) (bSelfMoved :type Boolean) (HitLocation :type Vector :default "0, 0, 0") (HitNormal :type Vector :default "0, 0, 0") (NormalImpulse :type Vector :default "0, 0, 0") (Hit :type Hit%20Result%20Structure%20%28by%20ref%29))
  (Audio|Components|Audio|Stop (Variables|Default|GetAudioInAirLoop))
  (|CallParentFunction MyComp Other OtherComp bSelfMoved HitLocation HitNormal NormalImpulse Hit))

(fn ConstructionScript
  (|CallParentFunction))
